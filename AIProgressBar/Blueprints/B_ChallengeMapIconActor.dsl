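; B_ChallengeMapIconActor is an actor that carries a minimap icon, named for challenge map markers, through its FortMiniMap component.
; SetMapIcon changes the icon only on authority, then flushes net dormancy so the change replicates to clients.
; BeginPlay also flushes dormancy after calling the parent.
; SetIconScale sets the icon scale on the same component, with no authority check or dormancy flush.
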(event EventBeginPlay
  ;[0.007568 0.06955 0.07812] BeginPlay: Replicate to the client.
  (|CallParentFunction)
  (Networking|FlushNetDormancy))

(fn SetIconScale
  (params (IconScale :type Vector%202D%20Structure :default "(X=1.500000,Y=1.500000)"))
  (MiniMapData|SetMiniMapIconScale (Variables|Default|GetFortMiniMapComponent) IconScale))

(fn SetMapIcon
  (params (MiniMapIcon :type Object%20Reference))
  ;[0.02466 0.1126 0.1198] SetMapIcon: Update the map icon.
  (Networking|SwitchHasAuthority
    (:Authority
      (MiniMapData|SetMiniMapIcon (Variables|Default|GetFortMiniMapComponent) MiniMapIcon)
      (Networking|FlushNetDormancy))))

(fn ConstructionScript)
