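; Challenge manager for the AI Progress Bar "god challenge", extending a base manager to run a staged living-world AI encounter with boss hooks.
; On playlist ready it self-destructs if hotfix-disabled, else picks encounter and goals by team size, reserves AI density and spawns a map icon actor.
; Advancing from Inactive starts the encounter after a camera-shake delay or earthquake.
; It also leashes spawned AI, toggles tagged point providers and routes phase music.
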(event EventStartChallenegeAfterDelay
  ;[0 0 0] Start Challenge (After SFX Delay)
  (Networking|SwitchHasAuthority
    (:Authority
      (bind (_isvalid _returnvalue) (Default|StartActivityEncounterStaticPointProvider (Fortnite|Athena|GameState|GetGameStateAthena) self (Variables|References|GetEncounter%20Type) (Variables|References|GetProviders)))
      (Variables|ChallengeManagerBase|SetEncounterRef _returnvalue)
      (|CallParentFunction))))

(event EventBeginPlay
  ;[0 1 0] BeginPlay
  (|CallParentFunction)
  (bind (_gamestate _playlist) (Class|FortAsyncActionCurrentPlaylistReady|PlaylistDataReadyAsync)
    (:OnReady
      (if (ScalableFloat|GetValueasBool (Variables|Hotfixes|GetRow_ChallengeEnabled))
        (Utilities|FlowControl|Sequence
          (:then_0
            (Milkshake|UpdateEarthquakeInfo)
            (AI|SetLagerDatafromPlaylistSize _gamestate _playlist)
            (CallFunction|Init))
          (:then_1
            ;;[1 1 0] Start debugging data
            (if (Variables|Debug|GetPrintDebugInformation)
              (bind _returnvalue (Utilities|Time|SetTimerbyEvent (EventDispatchers|CreateEvent :!event DebugGodChallenge) 0.5 true true))
              (Variables|Debug|SetDebugTimer _returnvalue))))
        (else
          ;;[1 0.25 0] Disabling Hotfix
          (Fortnite|FortShippingLog self "Destroying challenge manager due to being disabled in plugin data")
          (Actor|DestroyActor))))))

(event EventAdvanceChallenge
  ;[0.2 0.2 1] Advance Challenge - Starts the challenge if not started, go the next stage if possible or set the challenge as completed if not
  (|CallParentFunction)
  (switch Utilities|FlowControl|Switch|SwitchonEChallengeState (Variables|Progress|GetCurrentState)
    (:Inactive
      (if (ScalableFloat|GetValueasBool (Variables|Hotfixes|Earthquake|GetRow_EarthquakeEnabled))
        ;;[0 0 0] Start challenge with earthquake
        (Utilities|Time|SetTimerbyEvent (EventDispatchers|CreateEvent :!event StartChallengeWithEarthquake) (Math|Float|Clamp(Float) (ScalableFloat|GetValueatLevel (Variables|Hotfixes|Earthquake|GetRow_EarthquakeStartDelay)) 0.1 2e+01))
        (else
          ;;[0 0 0] Start challenge with delay
          (CallFunction|MultiPlayCameraShake)
          (Utilities|Time|SetTimerbyEvent (EventDispatchers|CreateEvent :!event StartChallenge) (Math|Float|Clamp(Float) (ScalableFloat|GetValueatLevel (Variables|Hotfixes|GetRow_StartChallengeDelay)) 0.1 2e+01)))))))

(event Custom|OnEarthquakeEnd
  (params (Actor :type Actor%20Object%20Reference) (EndPlayReason :type EEndPlayReason%20Enum :default "Destroyed"))
  ;[0.2 0.2 1] Advance Challenge - Starts the challenge if not started, go the next stage if possible or set the challenge as completed if not
  ;;[1 1 0] StartChallengeWithEarthquake binds to earthquake endplay on authority
  (Networking|SwitchHasAuthority
    (:Authority
      (Game|UnbindEventfromOnEndPlay Actor (EventDispatchers|CreateEvent :!event Custom|OnEarthquakeEnd))
      (CallFunction|StartChallenge))))

(event Custom|Init
  ;[0.133 1 0.133] Init - Reserve ai budget, calculate challenge goals, spawn and update map icon actor,
  ; ensure default challenge state and update EventAreaExitCollision size and location if not valid
  (Utilities|FlowControl|Sequence
    (:then_0
      (Networking|SwitchHasAuthority
        (:Authority
          (Utilities|FlowControl|Sequence
            (:then_0
              ;;[0 0 0] Reserver Density for AI
              (Class|FortAthenaLivingWorldManager|RegisterDensityReservoir (Class|FortAthenaLivingWorldManager|GetLivingWorldManager) self (ScalableFloat|GetValueatLevel (Variables|Hotfixes|GetRow_ReservoirDensity)) (ScalableFloat|GetValueatLevel (Variables|Hotfixes|GetRow_ReservoirRange)))
              ;;[0 0 0] Fill our PhaseProgressTargets array with data from StageData
              (Progress|CalculatePhaseProgressTargets)
              (Progress|CalculateTotalProgressNeededPerStage))
            (:then_1
              ;;[0 0 0] Create the MapIconActor and setup the initial texture. This is how HeistBases handled the icon. I am reusing this functionality because our challenges are currently implemented as level instances(we may have to change this) which do not play nicely with world partition and may not be loaded on clients and this would stop the icon from updating if they are not in relevancy.
              (bind _returnvalue (Game|SpawnActorfromClass "/AIProgressBar/Blueprints/B_ChallengeMapIconActor.B_ChallengeMapIconActor_C" (Math|Transform|MakeTransform (select (Utilities|Operators|NotEqual(!=) :A (Variables|Default|GetMapIconLocationOverride) :ErrorTolerance 0.001) (Variables|Default|GetMapIconLocationOverride) (Transformation|GetActorLocation))) "AlwaysSpawn" "MultiplyWithRoot" 0 0))
              (bind _output_get (Variables|MapIcon|SetMapIconActor _returnvalue))
              (Utilities|IsValid _output_get
                (:Is%20Valid
                  (Class|BChallengeMapIconActor|SetIconScale _output_get (Variables|MapIcon|GetMapIconScale)))))
            (:then_2
              (CallFunction|UpdateChallengeState "Inactive"))))))
    (:then_1
      (CallFunction|UpdateAreaExitCollision))))

(event Custom|Multi_PlayBuildUpSFX
  (params (Actor :type BP%20Milkshake%20Earthquake%20Object%20Reference))
  ;[1 1 1] TransitionEffects
  (Audio|PlaySoundatLocation (Variables|Sound|GetBossBuildUpCue) (Transformation|GetWorldLocation (Variables|/Milkshake/Gameplay/BP_Milkshake_Earthquake.BP_Milkshake_Earthquake_C|GetDefaultSceneRoot Actor)))
  (if (not (Variables|Sound|GetbPlayEarthquakeCue))
    (bind _earthquake_audio_component (Variables|/Milkshake/Gameplay/BP_Milkshake_Earthquake.BP_Milkshake_Earthquake_C|GetEarthquake%20Audio%20Component Actor)
      (:then
        (Audio|Components|Audio|FadeOut _earthquake_audio_component 0.1 0.0 "Linear" :!class /Script/Engine.AudioComponent))
      (:else))))

(event EventPlayTransitionEffects
  ;[1 1 1] TransitionEffects
  (if (ScalableFloat|GetValueasBool (Variables|Hotfixes|Earthquake|GetRow_EarthquakeForTransitionEnabled))
    ;;[0 0 0] Transition with earthquake fx
    (Milkshake|SpawnMilkShakeActor (Variables|Milkshake|GetTransitionEarthquakeInfo))
    (elif (ScalableFloat|GetValueasBool (Variables|Hotfixes|GetRow_CameraShakeForTransitionEnabled))
      ;;[0 0 0] Transition with camera shake
      (CallFunction|MultiPlayCameraShake))))

(event Custom|Multi_PlayCameraShake
  ;[1 1 1] TransitionEffects
  ;;[0 0 0] multicast world camera shake
  ;;;[1 1 0] Placeholder
  (Camera|PlayWorldCameraShake "/Game/Effects/Fort_Effects/BluePrints/CameraShakes/B_Large_Explosion_CameraShake.B_Large_Explosion_CameraShake_C" (Transformation|GetActorLocation) 2e+03 5e+03))

(event EventSetPhaseMusicEvent
  (bind _returnvalue (CallFunction|GetCurrentStageData))
  (bind (_goaltarget _phase) (Utilities|Struct|BreakAIProgressBarStageData _returnvalue))
  ;[0.4254 0.3272 1] Music Event Router
  ;;[0.7024 0.954 1] Only update music phase if we are advancing stages
  (bind (_updated_variable _previous_variable) (Utilities|FlowControl|DeltaGateWildcard (Variables|Music|GetMusicPhase) _phase)
    (:Success
      (if (> _updated_variable _previous_variable)
        (Variables|Music|SetMusicPhase _phase)))))

(event Custom|EnableOnlyInitialProviders
  ;[0 0 0] Events to flip flop which providers we have enabled.
  (AI|EnablePointProvidersbyTag (Variables|AI|GetInitialPointProviderTagName) false true))

(event Custom|EnableOnlyRandomProviders
  ;[0 0 0] Events to flip flop which providers we have enabled.
  (AI|EnablePointProvidersbyTag (Variables|AI|GetRandomPointProviderTagName) false true))

(event Custom|BroadcastPhaseMusicEvent
  (params (Phase :type Integer)))

(event Custom|StartActivationMusic
  ;[0.4254 0.3272 1] Music Event Router
  (Variables|Music|SetMusicPhase 0))

(event OnComponentBeginOverlap(EventAreaCollision)
  (params (OverlappedComponent :type Primitive%20Component%20Object%20Reference) (OtherActor :type Actor%20Object%20Reference) (OtherComp :type Primitive%20Component%20Object%20Reference) (OtherBodyIndex :type Integer) (bFromSweep :type Boolean) (SweepResult :type Hit%20Result%20Structure%20%28by%20ref%29))
  ;[0.2 0.2 1] Hooks for when a player pawn enters and exits the area
  (CallFunction|RequestPawnEnterArena OtherActor))

(event EventSetAIProgressBarChallengeRef
  (params (Actor :type Actor%20Object%20Reference) (bSetToSelf :type Boolean))
  ;[0.2 0.2 1] Hooks for when a player pawn enters and exits the area
  (bind _asfort_player_pawn_athena (Utilities|Casting|CastToFortPlayerPawnAthena Actor))
  (Utilities|IsValid (Actor|GetComponentbyClass _asfort_player_pawn_athena "/AIProgressBar/Blueprints/Components/BP_AIProgressBarPawnComponent.BP_AIProgressBarPawnComponent_C")
    (:Is%20Valid
      (if bSetToSelf
        (Variables|/AIProgressBar/Blueprints/Components/BP_AIProgressBarPawnComponent.BP_AIProgressBarPawnComponent_C|SetCurrentChallengeManager self (Actor|GetComponentbyClass _asfort_player_pawn_athena "/AIProgressBar/Blueprints/Components/BP_AIProgressBarPawnComponent.BP_AIProgressBarPawnComponent_C"))
        (else
          (Variables|/AIProgressBar/Blueprints/Components/BP_AIProgressBarPawnComponent.BP_AIProgressBarPawnComponent_C|SetCurrentChallengeManager 0 (Actor|GetComponentbyClass _asfort_player_pawn_athena "/AIProgressBar/Blueprints/Components/BP_AIProgressBarPawnComponent.BP_AIProgressBarPawnComponent_C")))))))

(event EventUpdateMapIconByState
  (params (ChallengeState :type EChallengeState%20Enum))
  ;[0 0 0] Update UI Icon On Map
  (switch Utilities|FlowControl|Switch|SwitchonEChallengeState ChallengeState
    (:None
      _)
    (:Inactive
      _)
    (:Active
      (if (ScalableFloat|GetValueasBool (Variables|Hotfixes|GetRow_DisableMapIconOnBossSpawn)) :!label _label
        _
        (else
          _)))
    (:Completed
      (goto _label)))
  (bind _mapiconactor (Variables|MapIcon|GetMapIconActor)
    (:then
      (Class|BChallengeMapIconActor|SetMapIcon _mapiconactor (Utilities|Select :Index ChallengeState :Inactive (Variables|MapIcon|GetInactiveMapIcon) :Active (Variables|MapIcon|GetActiveMapIcon) :Completed (Variables|MapIcon|GetCompletedMapIcon))))
    (:else)))

(event OnComponentEndOverlap(EventAreaExitCollision)
  (params (OverlappedComponent :type Primitive%20Component%20Object%20Reference) (OtherActor :type Actor%20Object%20Reference) (OtherComp :type Primitive%20Component%20Object%20Reference) (OtherBodyIndex :type Integer))
  ;[0.2 0.2 1] Hooks for when a player pawn enters and exits the area
  (CallFunction|RequestPawnExitArena 0))

(event EventHandleSpawnedPawn
  (params (EncounterInstance :type Fort%20Athena%20Living%20World%20Encounter%20Instance%20Object%20Reference) (SpawnedActor :type Actor%20Object%20Reference) (EventData :type Fort%20Athena%20Living%20World%20Event%20Data%20Object%20Reference))
  ;[0 0 0] HandleSpawnedPawn - Add pawn to list and bind to
  ; death event for tracking and progress.
  ;;[0.2 0.2 1] Update Leash
  (Leash|UpdateAIActorLeash SpawnedActor)
  ;;[0.2 0.2 1] Add music tag if valid
  (Music|AddCurrentMusicTagtoActor SpawnedActor))

(event EventDisableMapIcon
  ;[0 0 0] Update UI Icon On Map
  ;;[0 0 0] Disable the UI Map Icon
  (Networking|SwitchHasAuthority
    (:Authority
      (if (ScalableFloat|GetValueasBool (Variables|Hotfixes|GetRow_DisableMapIconOnBossSpawn))
        (bind _mapiconactor (Variables|MapIcon|GetMapIconActor)
          (:then
            (Actor|DestroyActor _mapiconactor))
          (:else))))))

(event EventBindToBossPawnEvents
  ;[0 0 0] Bindings for Boss Pawn Events
  (bind _asbp_player_pawn_non_participant_boss (Utilities|Casting|CastToBP_PlayerPawn_NonParticipant_Boss (Variables|Boss|GetBossActor)))
  (Health|AssignOnPawnHealthChanged _asbp_player_pawn_non_participant_boss (EventDispatchers|CreateEvent :!event HandleBossHealthChanged))
  (Default|BindEventtoOnBossReadyforFinalFight _asbp_player_pawn_non_participant_boss (EventDispatchers|CreateEvent :!event SetIsBossReadyForFinalFight)))

(event EventUnbindToBossPawnEvents
  ;[0 0 0] Bindings for Boss Pawn Events
  (bind _asbp_player_pawn_non_participant_boss (Utilities|Casting|CastToBP_PlayerPawn_NonParticipant_Boss (Variables|Boss|GetBossActor)))
  (Health|UnbindEventfromOnPawnHealthChanged _asbp_player_pawn_non_participant_boss (EventDispatchers|CreateEvent :!event HandleBossHealthChanged))
  (Default|UnbindEventfromOnBossReadyforFinalFight _asbp_player_pawn_non_participant_boss (EventDispatchers|CreateEvent :!event SetIsBossReadyForFinalFight)))

(event EventEndPlay
  (params (EndPlayReason :type EEndPlayReason%20Enum :default "Destroyed"))
  ;[1 0.25 0] EndPlay - Clear debug timer, pause encounter
  (|CallParentFunction EndPlayReason)
  (Utilities|Time|ClearandInvalidateTimerbyHandle (Variables|Debug|GetDebugTimer))
  (Networking|SwitchHasAuthority
    (:Authority
      (CallFunction|CleanupEncounter))))

(event EventResetProgress
  ;[1 0.25 0] EndPlay - Clear debug timer, pause encounter
  (|CallParentFunction)
  (Variables|Music|SetMusicPhase 0))

(event Custom|ResetGodChallenge
  (params (ChallengeName :type String))
  ;[0.4072 0.4072 0] (GodChallenge) Specifics
  (Networking|SwitchHasAuthority
    (:Authority
      (if (Utilities|String|Equal,CaseInsensitive(String) ChallengeName (Utilities|String|Replace (GameplayTags|GetDebugStringfromGameplayTag (Variables|Challenge|GetChallengeIDTag)) "ChallengeTheGods."))
        (CallFunction|CleanupEncounter)
        (CallFunction|UpdateChallengeState "Inactive")
        (Utilities|FlowControl|Delay (Variables|Debug|GetDebugResetChallengeDelay)) ; Allow for ai to cleanup
        (CallFunction|AdvanceChallenge)))))

(event Custom|SendAllAIActorToAlertLocation
  ;[0.4072 0.4072 0] (GodChallenge) Specifics
  ;;[1 1 0] (GodChallenge) Alert all AI towards the Altar - Currently only used for debug
  (Networking|SwitchHasAuthority
    (:Authority
      (bind _encounterref (Variables|ChallengeManagerBase|GetEncounterRef)
        (:then
          (for _array_element (Variables|/Script/LagerRuntime.FortAthenaLivingWorldEncounterInstance|GetSpawnedActors _encounterref)
            (AI|SendAIActortoAlertLocation _array_element)))
        (:else)))))

(event Custom|DebugGodChallenge
  ;[1 1 0 1 30] !!! Debug
  (switch Utilities|FlowControl|Switch|SwitchonEChallengeState (Variables|Progress|GetCurrentState)
    (:Active
      (bind _encounterref (Variables|ChallengeManagerBase|GetEncounterRef)
        (:then
          (Development|PrintString (Utilities|String|Append "Encounter stage: " (Utilities|String|ToString(Integer) (Variables|/Script/LagerRuntime.FortAthenaLivingWorldEncounterInstance|GetCurrentStageIndex _encounterref)) "
Stage Progress: " (Utilities|String|ToString(Integer) (Variables|Progress|GetTotalProgress)) "/" (Utilities|String|ToString(Integer) (CallFunction|GetStageTotalProgressGoal (Variables|Progress|GetCurrentStage))) "
Current phase: " (Utilities|String|ToString(Integer) (Variables|Progress|GetCurrentPhase)) "
Phase Progress: " (Utilities|String|ToString(Integer) (Variables|Progress|GetCurrentPhaseProgress)) "/" (Utilities|String|ToString(Integer) (CallFunction|GetCurrentPhaseProgressTarget))) true false "(R=1.000000,G=0.564936,B=0.000000,A=1.000000)" 1.0 "GodChallengeDebug"))
        (:else)))))

(event Custom|DebugDrawLeashRadius
  (bind _leashlocation (Leash|GetLeashLocation))
  (bind _debugdrawleashduration (Variables|Debug|GetDebugDrawLeashDuration))
  ;[1 1 0] Draw default leash spheres
  (if (Editor|InEditororPIE)
    (Rendering|Debug|DrawDebugSphere _leashlocation (ScalableFloat|GetValueatLevel (Variables|Hotfixes|GetRow_LeashInnerRadius)) 16 "(R=0.000000,G=1.000000,B=0.000000,A=1.000000)" _debugdrawleashduration 1e+01)
    (Rendering|Debug|DrawDebugSphere _leashlocation (ScalableFloat|GetValueatLevel (Variables|Hotfixes|GetRow_LeashOuterRadius)) 32 "(R=1.000000,G=0.250000,B=0.000000,A=1.000000)" _debugdrawleashduration 15.0)))

(event Custom|ToggleProvidersEnabled
  (Networking|SwitchHasAuthority
    (:Authority
      (bind _isa (Utilities|FlowControl|FlipFlop)
        (:A
          _)
        (:B
          _))
      (AI|EnablePointProvidersbyTag (select _isa (Variables|AI|GetInitialPointProviderTagName) (Variables|AI|GetRandomPointProviderTagName)) true true))))

(fn AddCurrentMusicTagToActor
  (params (Actor :type Actor%20Object%20Reference))
  (bind _currentphasemusictag (Variables|Content|GetCurrentPhaseMusicTag))
  (if (GameplayTags|IsGameplayTagValid _currentphasemusictag)
    (bind _returnvalue (GameplayTags|HasMatchingGameplayTag(Message) Actor _currentphasemusictag))
    (if (not _returnvalue)
      (Ability|GameplayEffect|AddLooseGameplayTags Actor (GameplayTags|MakeGameplayTagContainerfromTag _currentphasemusictag) true))))

(fn AdvanceLivingWorldEncounterStage
  (Networking|SwitchHasAuthority
    (:Authority
      (bind _encounterref (Variables|ChallengeManagerBase|GetEncounterRef)
        (:then
          (Class|FortAthenaLivingWorldEncounterInstance|AdvanceEncounterStage _encounterref))
        (:else)))))

(fn CalculatePhaseProgressTargets
  (bind _phaseprogresstargets (Variables|Progress|GetPhaseProgressTargets))
  (bind _goaltargetindexbyplaylist (Variables|Progress|GetGoalTargetIndexByPlaylist))
  (bind _phasetoaddto (LocalVariables|GetPhaseToAddTo))
  (Utilities|Array|Clear (Variables|Progress|GetPhaseProgressTargets))
  (for _array_element (Variables|Progress|GetStageData)
    ;[0 0 0] What phase of the challenge are we looking at right now.
    ; Is this a valid phase?
    (bind _output_get (LocalVariables|SetPhaseToAddTo (.Phase (Utilities|Struct|BreakAIProgressBarStageData _array_element))))
    (if (> _output_get -1)
      (Utilities|FlowControl|Sequence
        (:then_0
          ;[0 0 0] Check to see if this stage has a goal specified per max team size of the playlist.
          (if (Utilities|Array|IsValidIndex (Utilities|Struct|BreakAIProgressBarStageData _array_element) _goaltargetindexbyplaylist)
            (LocalVariables|SetGoal (Utilities|Array|Get(acopy) (Utilities|Struct|BreakAIProgressBarStageData _array_element) _goaltargetindexbyplaylist)) ; If found set to value
            (else
              (LocalVariables|SetGoal 1) ; If not found set to 1
            )))
        (:then_1
          ;[0 0 0] Check to see if we have a a goal for this phase and if so add it to our local var
          (if (Utilities|Array|IsValidIndex _phaseprogresstargets _phasetoaddto)
            (LocalVariables|SetGoal (+ (LocalVariables|GetGoal) (Utilities|Array|Get(acopy) _phaseprogresstargets _phasetoaddto)))))
        (:then_2
          ;[0 0 0] Update the goal for the phase we are working on
          (Utilities|Array|SetArrayElem (Variables|Progress|GetPhaseProgressTargets) (LocalVariables|GetPhaseToAddTo) (LocalVariables|GetGoal) true))))))

(fn CalculateTotalProgressNeededPerStage
  (for (_array_element _array_index) (Variables|Progress|GetStageData)
    _))

(fn DespawnAllMinionsBP
  (LocalVariables|SetpriorCountDeathBool (Variables|Default|GetCountMinionDeathForProgress))
  (Variables|Default|SetCountMinionDeathForProgress false)
  (bind _encounterref (Variables|ChallengeManagerBase|GetEncounterRef)
    (:then
      (for _array_element (Variables|/Script/LagerRuntime.FortAthenaLivingWorldEncounterInstance|GetSpawnedActors _encounterref)
        (bind _asfort_pawn (Utilities|Casting|CastToFortPawn _array_element))
        (if (not (GameplayTags|HasMatchingGameplayTag _asfort_pawn (Variables|Boss|GetBossIDTag)))
          (FortPawn|ForceKill :self _asfort_pawn)))
      (Variables|Default|SetCountMinionDeathForProgress (LocalVariables|GetpriorCountDeathBool)))
    (:else)))

(fn DisableProvidersByTags
  (params (TagsToDisable :type Gameplay%20Tag%20Container%20Structure%20%28by%20ref%29))
  (for _array_element (Variables|References|GetProviders)
    (if (GameplayTags|HasAnyTags (Class|FortAthenaLivingWorldPointProviderInterface|GetFiltersTags _array_element) TagsToDisable true)
      (Class|FortAthenaLivingWorldPointProviderInterface|DisablePointProvider _array_element))))

(fn EnableAllPointProviders
  (Networking|SwitchHasAuthority
    (:Authority
      (for _array_element (Variables|References|GetProviders)
        (Utilities|IsValid _array_element
          (:Is%20Valid
            (Class|FortAthenaLivingWorldPointProviderInterface|EnablePointProvider_DEPRECATED _array_element)))))))

(fn EnablePointProvidersByTag
  (params (TagToCheck :type Name) (ShouldDisable :type Boolean) (SetNonMatchingToOpposite :type Boolean))
  (Networking|SwitchHasAuthority
    (:Authority
      (for _array_element (Variables|References|GetProviders)
        (Utilities|IsValid _array_element
          (:Is%20Valid
            (if (not (Actor|ActorHasTag _array_element (Variables|AI|GetBossPointProviderTagName)))
              (if (Actor|ActorHasTag _array_element (Variables|GetTagToCheck))
                ;[0 0 0] Enable/Disable the provider
                (if (Variables|GetShouldDisable)
                  (Class|FortAthenaLivingWorldStaticPointProvider|DisablePointProvider _array_element)
                  (else
                    (Class|FortAthenaLivingWorldPointProviderInterface|EnablePointProvider_DEPRECATED _array_element)))
                (elif (Variables|GetSetNonMatchingToOpposite)
                  ;[0 0 0] Option to set all providers that don't match the tag to the opposite value
                  (if (Variables|GetShouldDisable)
                    (Class|FortAthenaLivingWorldPointProviderInterface|EnablePointProvider_DEPRECATED _array_element)
                    (else
                      (Class|FortAthenaLivingWorldStaticPointProvider|DisablePointProvider _array_element))))))))))))

(fn GetLeashLocation
  ;[0.5156 0.5156 0.5156] Would like to add an instance editable vector and if it is set then we use that location instead. Also include a flag to toggle debug sphere drawing so that LD/Gameplay can easily see the size of the leash
  (bind _leashlocationactor (Variables|Leash|GetLeashLocationActor)
    (:then
      (return (Transformation|GetActorLocation _leashlocationactor)))
    (:else
      (return (Transformation|GetActorLocation)))))

(fn GetSummoningActor
  (return (Variables|References|GetSummoningActor)))

(fn IsLastPhase
  (params (PhaseIndex :type Integer))
  (bind (_output_goaltarget _output_phase) (Utilities|Struct|BreakAIProgressBarStageData (Utilities|Array|Get(aref) (Variables|Progress|GetStageData) (- (Utilities|Array|Length (Variables|Progress|GetStageData)) 1))))
  (return (== PhaseIndex _output_phase)))

(fn OnRep_MusicPhase
  (CallFunction|BroadcastPhaseMusicEvent (Variables|Music|GetMusicPhase)))

(fn SendAIActorToAlertLocation
  (params (AIActor :type Actor%20Object%20Reference))
  (Networking|SwitchHasAuthority
    (:Authority
      (Utilities|IsValid AIActor
        (:Is%20Valid
          ;[0 0 0] Don't ever do for bosses
          (bind _returnvalue (GameplayTags|HasMatchingGameplayTag(Message) AIActor (Variables|Boss|GetBossIDTag)))
          (if (not _returnvalue)
            ;[0 0 0] Get altar location actor and send a stim to AI to get them to move towards that location with some deviation
            (bind _alertlocationactor (Variables|References|GetAlertLocationActor)
              (:then
                (bind _asfort_athena_aibot_controller (Utilities|Casting|CastToFortAthenaAIBotController (Fortnite|GetControllerfromActor AIActor)))
                (Class|FortAthenaAIBotController|AddTargetInfo _asfort_athena_aibot_controller _alertlocationactor false "Threat_LKP" (ScalableFloat|GetValueatLevel (Variables|Hotfixes|GetRow_AIAlertLocationForgetTime)) (ScalableFloat|GetValueatLevel (Variables|Hotfixes|GetRow_AIAlertLocationForgetDistance))))
              (:else
                ;[1 0 0] Log that the actorlocation is not valid
                (Fortnite|FortShippingLog self "Failed to send AI to AlertLocationActor. Actor is not valid." true)))))))))

(fn SetLagerDataFromPlaylistSize
  (params (InWorldContext :type Object%20Reference) (Playlist :type Fort%20Playlist%20Object%20Reference))
  (Networking|SwitchHasAuthority
    (:Authority
      (bind _asfort_playlist_athena (Utilities|Casting|CastToFortPlaylistAthena Playlist))
      (Variables|References|SetEncounter%20Type (Utilities|Array|Get(acopy) (Variables|References|GetEncounterTypesByPlaylistSize) (- (Playlist|GetMaxTeamSize InWorldContext _asfort_playlist_athena) 1)))
      (Variables|Progress|SetGoalTargetIndexByPlaylist (- (Playlist|GetMaxTeamSize InWorldContext _asfort_playlist_athena) 1)))))

(fn SetSummoningActor
  (params (SummoningActor :type Actor%20Object%20Reference))
  (Variables|References|SetSummoningActor SummoningActor))

(fn SpawnMilkShakeActor
  (params (MilkeShakeInfoStruc :type Milkshake%20Earthquake%20Info%20Struct%20Structure))
  (return 0))

(fn StartChallengeWithEarthquake
  (Networking|SwitchHasAuthority
    (:Authority
      (bind _earthquakeactor (Milkshake|SpawnMilkShakeActor (Variables|Milkshake|GetChallengeStartEarthquakeInfo)))
      (bind _output_get (Variables|Milkshake|SetEarthquakeActor _earthquakeactor))
      (Utilities|IsValid _output_get
        (:Is%20Valid
          (CallFunction|MultiPlayBuildUpSFX _output_get)
          (Game|BindEventtoOnEndPlay _output_get (EventDispatchers|CreateEvent :!event OnEarthquakeEnd)))))))

(fn UpdateAreaExitCollision
  (bind _returnvalue (Components|Box|GetUnscaledBoxExtent (Variables|Default|GetEventAreaCollision)))
  (bind (_returnvalue_x _returnvalue_y _returnvalue_z) (Math|Vector|BreakVector (- (Components|Box|GetUnscaledBoxExtent (Variables|Default|GetEventAreaExitCollision)) _returnvalue)))
  (bind _eventareacollision (Variables|Default|GetEventAreaCollision))
  (bind _eventareaexitcollision (Variables|Default|GetEventAreaExitCollision))
  (bind _relativelocation (Variables|/Script/Engine.SceneComponent|GetRelativeLocation _eventareacollision))
  (Utilities|FlowControl|Sequence
    (:then_0
      ;[0 0 0] Make the exit box extents 1000 units larger than the entrance box if they are not already set to be larger by 10 units
      (if (Math|Boolean|ORBoolean (< _returnvalue_x 1e+01) (< _returnvalue_y 1e+01) (< _returnvalue_z 1e+01))
        (Components|Box|SetBoxExtent _eventareaexitcollision (+ _returnvalue (ScalableFloat|GetValueatLevel (Variables|Hotfixes|GetRow_AreaExitCollisionExtentIncrease))) true)))
    (:then_1
      ;[0 0 0] If Exit box location is not near the entrance box move it to the same location
      (if (Utilities|Operators|NotEqual(!=) _relativelocation (Variables|/Script/Engine.SceneComponent|GetRelativeLocation _eventareaexitcollision) 1e+01)
        (Transformation|SetRelativeLocation _eventareaexitcollision _relativelocation false false)))))

(fn UpdateAIActorLeash
  (params (AIActor :type Actor%20Object%20Reference))
  (bind _row_leashouterradius (Variables|Hotfixes|GetRow_LeashOuterRadius))
  (bind _returnvalue (ScalableFloat|GetValueatLevel _row_leashouterradius))
  (bind _row_leashinnerradius (Variables|Hotfixes|GetRow_LeashInnerRadius))
  (bind _returnvalue_2 (ScalableFloat|GetValueatLevel _row_leashinnerradius))
  (Utilities|IsValid AIActor
    (:Is%20Valid
      (if (and (> _returnvalue_2 0) (> _returnvalue 0))
        (bind _asfort_athena_aibot_controller (Utilities|Casting|CastToFortAthenaAIBotController (Fortnite|GetControllerfromActor AIActor)))
        (Class|BPFLAICommonFunctions|SetRadialLeash _asfort_athena_aibot_controller (Leash|GetLeashLocation) _returnvalue_2 _returnvalue)))))

(fn UpdateEarthquakeInfo
  (Networking|SwitchHasAuthority
    (:Authority
      (Utilities|Struct|SetmembersinMilkshakeEarthquakeInfoStruct (Variables|Milkshake|GetChallengeStartEarthquakeInfo) (Math|Vector2D|MakeVector2D (ScalableFloat|GetValueatLevelwithContext (Variables|Hotfixes|Earthquake|GetRow_EarthquakeDurationMin)) (ScalableFloat|GetValueatLevelwithContext (Variables|Hotfixes|Earthquake|GetRow_EarthquakeDurationMax))) (Math|Vector2D|MakeVector2D (ScalableFloat|GetValueatLevelwithContext (Variables|Hotfixes|Earthquake|GetRow_EarthquakeRadiusMin)) (ScalableFloat|GetValueatLevelwithContext (Variables|Hotfixes|Earthquake|GetRow_EarthquakeRadiusMax)))))))

(fn UpdateMusicTagsOnSpawnedActors
  (bind _encounterref (Variables|ChallengeManagerBase|GetEncounterRef)
    (:then
      (Utilities|FlowControl|Sequence
        (:then_0
          ;[0 0 0] Setup local  variable for tags to remove so we dont have to repeat logic in loop.
          (LocalVariables|SetTagsToRemove (GameplayTags|MakeGameplayTagContainerfromArray (Variables|Content|GetPhaseMusicTags))))
        (:then_1
          (for _array_element (Variables|/Script/LagerRuntime.FortAthenaLivingWorldEncounterInstance|GetSpawnedActors _encounterref)
            (Utilities|IsValid _array_element
              (:Is%20Valid
                (Ability|GameplayEffect|RemoveLooseGameplayTags _array_element (LocalVariables|GetTagsToRemove) true) ; Remove any music tags
                (Music|AddCurrentMusicTagtoActor _array_element)))))))
    (:else)))

(fn ConstructionScript
  (|CallParentFunction))
