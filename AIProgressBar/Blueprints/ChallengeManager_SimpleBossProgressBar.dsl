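; Boss challenge manager that runs a single AI boss fight as a progress-bar challenge with an arena area.
; Once playlist data is ready, it calculates progress targets, enables a patrol point provider and listens for its spawn.
; The spawned boss starts the challenge and gets leash and music-tag setup; its death completes progress and spawns a reward.
; Area overlaps request pawn arena entry, while a separate exit volume's end-overlap requests exit.
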(event EventBeginPlay
  (Class|FortAsyncActionCurrentPlaylistReady|PlaylistDataReadyAsync
    (:OnReady
      (CallFunction|Init))))

(event EventInit
  (bind _patrolpointprovider (Variables|References|GetPatrolPointProvider))
  (Utilities|FlowControl|Sequence
    (:then_0
      (Progress|CalculatePhaseProgressTargets)
      (Progress|CalculateTotalProgressNeededPerStage)
      (Utilities|IsValid _patrolpointprovider
        (:Is%20Valid
          (EventDispatchers|BindEventtoOnActorSpawnedResult _patrolpointprovider (delegate Custom|AISpawned))
          (Class|FortAthenaLivingWorldPointProviderInterface|EnablePointProvider_DEPRECATED _patrolpointprovider))))
    (:then_1
      (CallFunction|UpdateAreaExitCollision))))

(event Custom|AISpawned
  (params (SpawnedActor :type Actor%20Object%20Reference) (bResult :type Boolean))
  (CallFunction|HandleBossSpawned SpawnedActor)
  (CallFunction|SetIsBossReadyforFinalFight true)
  (CallFunction|StartSimpleChallenge)
  (Leash|UpdateAIActorLeash SpawnedActor)
  (Music|AddCurrentMusicTagtoActor SpawnedActor)
  (bind _asfort_player_pawn_athena (Utilities|Casting|CastToFortPlayerPawnAthena SpawnedActor))
  (bind _asfort_athena_aibot_controller (Utilities|Casting|CastToFortAthenaAIBotController (Pawn|GetController _asfort_player_pawn_athena)))
  (Death|BindEventtoOnAgentDiedEvent _asfort_athena_aibot_controller (delegate Custom|AIDied)))

(event OnComponentBeginOverlap(EventAreaCollision)
  (params (OverlappedComponent :type Primitive%20Component%20Object%20Reference) (OtherActor :type Actor%20Object%20Reference) (OtherComp :type Primitive%20Component%20Object%20Reference) (OtherBodyIndex :type Integer) (bFromSweep :type Boolean) (SweepResult :type Hit%20Result%20Structure%20%28by%20ref%29))
  (CallFunction|RequestPawnEnterArena OtherActor))

(event Custom|AIDied
  (params (BotController :type Fort%20Athena%20AIBot%20Controller%20Object%20Reference) (AgentPawn :type Fort%20Pawn%20Object%20Reference))
  (Variables|Progress|SetTotalProgress 1)
  (CallFunction|UpdateChallengeState "Completed")
  (CallFunction|SpawnReward AgentPawn)
  (CallFunction|HandleBossDied))

(event OnComponentEndOverlap(EventAreaExitCollision)
  (params (OverlappedComponent :type Primitive%20Component%20Object%20Reference) (OtherActor :type Actor%20Object%20Reference) (OtherComp :type Primitive%20Component%20Object%20Reference) (OtherBodyIndex :type Integer))
  (CallFunction|RequestPawnExitArena OtherActor))

(fn ConstructionScript
  (|CallParentFunction))
